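# Niagara skydiving contrail "Turbo": blue glowing effects attached to the character's bones during freefall.
# Short looping mesh bursts form particle-ID-indexed grids at hand and foot bones, layered with glow sprites, lights, streak rays and rising square particles.
# Spawn groups select bones on user-supplied skeletal meshes and mask fixed offsets for the squares.
# Owner velocity drives ray length and square speed, while FrontEnd modifiers switch tint and sizing between menu and gameplay.
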NiagaraSystem /BRCosmetics/Athena/Effects/Particles/Skydiving/NS_Contrail_Turbo/FX/NS_Contrail_Turbo
dependencies {
 usedDataInterfaces [&NiagaraDataInterfaceCurve, &NiagaraDataInterfaceColorCurve]
 usedDynamicInputs []
 usedModules [&SystemState]
}
emitters {
 FN_Contrails_AttachedParticle001 {
  emitterUpdateScript [
   &EmitterState(Inactive Response=::NewEnumerator1, Life Cycle Mode=::NewEnumerator1, Loop Duration=0, UseLoopDelay=false)
   &SpawnRate(SpawnRate=10, Use Spawn Probability=false)
  ]
  particleSpawnScript [
   &FN_SetSpawnGroup_v2(Number of Spawn Groups=4)
   /Niagara/Modules/Spawn/Initialization/InitializeParticle(
    Lifetime=0.6
    Module.Write Color=true
    Module.Write Lifetime=true
    Module.Write Mass=true
    Module.Write Material Random=true
    Module.Write Position=false
    Module.Write Scale=false
    Module.Write SpriteRotation=false
    Module.Write SpriteSize=true
    Position=&SimulationPosition()
    Sprite Size=&Vector2DFromFloat(Value=&UniformRangedFloat(Minimum=100, Maximum=150))
   )
   &CameraOffset()
   /SetVariables(Particles.LightRadius=60) {isSetParameters=true}
  ]
  particleUpdateScript [
   /SetVariables(Particles.Color=&MakeLinearColorFromVectorAndFloat(
     "Vector (RGB)"=&MakeVectorFromLinearColorRGB(Color=<rgba 0.0317 0.3512 1 1>)
     "Float (Alpha)"=&FloatFromCurve(
      FloatCurve=DataInterface() {propertyValues {Curve {keys [<at 0 0 RCIM_Cubic>, {arriveTangent=0.002, interpMode=RCIM_Cubic, leaveTangent=0.002, time=0.4993, value=0.5}, <at 1.0011 0.002>]}}}
      Scale Curve=1.2
     )
    )) {enabled=false, isSetParameters=true, module=SetVariables_#1}
   /SetVariables(Particles.Color=&MakeLinearColorFromVectorAndFloat(
     "Vector (RGB)"=&MakeVectorFromLinearColorRGB(Color=&ColorFromCurve(ColorCurve=DataInterface() {propertyValues {
         AlphaCurve {keys [<at -0.0023 1 RCIM_Cubic>,<at 0.9977 1 RCIM_Cubic>]}
         BlueCurve {keys [<at -0.0036 1 RCIM_Cubic>,<at 0.341 1>,<at 0.5867 1>,<at 1.0015 0.9502 RCIM_Cubic>]}
         GreenCurve {keys [<at -0.0036 0.3902 RCIM_Cubic>,<at 0.341 0.8803>,<at 0.5867 0.5665>,<at 1.0015 1 RCIM_Cubic>]}
         RedCurve {keys [<at -0.0036 0.027 RCIM_Cubic>,<at 0.341 0.0232>,<at 0.5867 0.031>,<at 1.0015 0.0299 RCIM_Cubic>]}
        }}))
     "Float (Alpha)"=&FloatFromCurve(
      FloatCurve=DataInterface() {propertyValues {Curve {keys [<at 0 0 RCIM_Cubic>, {arriveTangent=0.002, interpMode=RCIM_Cubic, leaveTangent=0.002, time=0.4963, value=0.15}, <at 1.0011 0.002>]}}}
      Scale Curve=1.2
     )
    )) {isSetParameters=true}
   /SetVariables(Particles.Light_Color:LinearColor=&MakeLinearColorFromVectorAndFloat(
     "Vector (RGB)"=&MakeVectorFromLinearColorRGB(Color=&ColorFromCurve(ColorCurve=DataInterface() {propertyValues {
         AlphaCurve {keys [<at 0 1 RCIM_Cubic>,<at 1 1 RCIM_Cubic>]}
         BlueCurve {keys [<at 0 6 RCIM_Cubic>,<at 0.5047 6>,<at 1.0023 6 RCIM_Cubic>]}
         GreenCurve {keys [<at 0 1.3451 RCIM_Cubic>,<at 0.5047 3.9199>,<at 1.0023 1.153 RCIM_Cubic>]}
         RedCurve {keys [<at 0 0.2439 RCIM_Cubic>,<at 0.5047 0.3362>,<at 1.0023 0.2701 RCIM_Cubic>]}
        }}))
     "Float (Alpha)"=&FloatFromCurve(FloatCurve=DataInterface() {propertyValues {Curve {keys [
          {arriveTangent=4.4866, interpMode=RCIM_Cubic, leaveTangent=4.4866, tangentMode=RCTM_User, time=-0.0002, value=-0.006}
          {arriveTangent=0.006, interpMode=RCIM_Cubic, leaveTangent=0.006, time=0.5038, value=1.0023}
          {arriveTangent=-4.3239, interpMode=RCIM_Cubic, leaveTangent=-4.3239, tangentMode=RCTM_User, time=1, value=0}
         ]}}})
    )) {isSetParameters=true}
   &ParticleState()
   &SolveForcesAndVelocity(Clamp Velocity=false, Limit Acceleration=false)
   &ScaleMeshSize(Scale Factor=&Multiply_Vector(
     A=<1 1 1>
     B=&VectorFromFloat(Value=&FloatFromCurve(FloatCurve=DataInterface() {propertyValues {Curve {keys [<at 0 0>,<at 1 1.5>]}}}))
    ))
   &SkeletalMeshLocation(
    Bone / Socket Index=$Particles.SpawnGroup
    Bone Sampling Mode=::NewEnumerator4
    Inherit Velocity=false
    Interpolate Mesh Sampling=false
    MaskSpawnGroup=true
    Sampled Orientation Coordinate Space=::World
    Sampled Position Offset Coordinate Space:ENiagaraMeshTransforms=::NewEnumerator2
    Skeletal Mesh=$User.SkeletalMesh
   ) {module=SkeletalMeshLocation002}
  ]
  rendererClasses [&NiagaraSpriteRendererProperties, &NiagaraLightRendererProperties]
  renderers [
   {rendererClass=&NiagaraSpriteRendererProperties}
   {rendererClass=&NiagaraLightRendererProperties, rendererIndex=1}
  ]
 }
 Foot_Big {
  emitterUpdateScript [
   &EmitterState(Inactive Response=::NewEnumerator1, Life Cycle Mode=::NewEnumerator1, Loop Duration=0.15, UseLoopDelay=false)
   &SpawnBurst_Instantaneous(Spawn Count=18, Use Spawn Probability=false)
  ]
  particleSpawnScript [
   /SetVariables(
    Particles.X_Grid:NiagaraFloat=&Lerp_Float(
     A=0
     B=20
     Alpha=&Frac_Float(Float=&Divide_Float(A=&MakeFloatFromInt(Int=/Niagara/DynamicInputs/Execution/ReturnParticleID()), B=3))
    )
    Particles.Y_Grid:NiagaraFloat=&Multiply_Float(
     A=7
     B=/Niagara/DynamicInputs/Math/Floor_Float(Float:NiagaraFloat=&Divide_Float(A=&MakeFloatFromInt(Int=&ReturnParticleID()), B=6))
    )
   ) {isSetParameters=true, module=SetVariables_#2}
   &InitializeParticle(
    Color=&MakeCustomLinearColorFromBool(
     Bool=/Niagara/DynamicInputs/Random/RandomBool(Randomness Mode:ENiagaraRandomnessMode=::NewEnumerator0, Random Seed:NiagaraInt32=0, Fixed Random Seed:NiagaraBool=false)
     True Color=&MakeCustomLinearColorFromBool(
      Bool=&RandomBool(Randomness Mode:ENiagaraRandomnessMode=::NewEnumerator0, Random Seed:NiagaraInt32=0, Fixed Random Seed:NiagaraBool=false)
      True Color=<rgba 0.0344 1.6317 8 2>
      False Color=<rgba 0 0.6138 4 2>
     )
     False Color=<rgba 0.2604 3.093 15 2>
    )
    Lifetime=&UniformRangedFloat(Minimum=0.15, Maximum=0.15)
    Mesh Scale=&Multiply_Vector(
     A=<0.7 0.7 0.7>
     B=&VectorFromFloat(Value=&UniformRangedFloat(
       Minimum=&Multiply_Float(
        A=0.4
        B=&FloatFromCurve(
         FloatCurve=DataInterface() {propertyValues {Curve {keys [<at 0 1>,<at 1 1>]}}}
         CurveIndex=&Divide_Float(A=&VectorLength(Vector=$Engine.Owner.Velocity), B=9000)
        )
       )
       Maximum=&Multiply_Float(
        A=0.55
        B=&FloatFromCurve(
         FloatCurve=DataInterface() {propertyValues {Curve {keys [<at 0 1>,<at 1 1>]}}}
         CurveIndex=&Divide_Float(A=&VectorLength(Vector=$Engine.Owner.Velocity), B=9000)
        )
       )
      ))
    )
    Module.Write Color=true
    Module.Write Lifetime=true
    Module.Write Mass=true
    Module.Write Material Random=true
    Module.Write Position=false
    Module.Write Scale=true
    Module.Write SpriteRotation=true
    Module.Write SpriteSize=true
    Position=&SimulationPosition()
    Sprite Size=&Vector2DFromFloat(Value=0)
   )
   &FN_SetSpawnGroup_v2(Number of Spawn Groups=2)
  ]
  particleUpdateScript [
   &ParticleState()
   &SolveForcesAndVelocity(Clamp Velocity=false, Limit Acceleration=false)
   &ScaleColor(
    Scale RGB=&FN_FrontEndVector3Modifier(
     Gameplay=&MakeVectorFromLinearColorRGB(Color=<rgba 0 0.5347 0.7 1>)
     Front End=&MakeVectorFromLinearColorRGB(Color=$User.FrontEnd_color)
    )
    ScaleA=true
    ScaleRGB=true
   )
   &ScaleMeshSize(Scale Factor=&VectorFromFloat(Value=&FloatFromCurve(FloatCurve=DataInterface() {propertyValues {Curve {keys [<at -0.0019 0.5372>,<at 0.2471 0.5372>,<at 0.4981 0.7>,<at 0.7498 0.5456>,<at 1.0019 0.5372>]}}})))
   &SkeletalMeshLocation(
    Bone / Socket Index=$Particles.SpawnGroup
    Bone Sampling Mode=::NewEnumerator4
    Inherit Velocity=false
    Interpolate Mesh Sampling=false
    MaskSpawnGroup=false
    Sampled Orientation Coordinate Space=::Local
    Sampled Position Offset=&Add_Vector(A=&MakeVector(X=$Particles.X_Grid, Y=$Particles.Y_Grid, Z=0), B=<-7 -3 0>)
    Sampled Position Offset Coordinate Space:ENiagaraMeshTransforms=::NewEnumerator2
    Skeletal Mesh=$User.SKeletal_Mesh_Foot
   )
   &ScaleMeshSizeBySpeed()
   &CameraOffset(Camera Offset Amount=30)
  ]
  rendererClasses [&NiagaraMeshRendererProperties]
  renderers [{rendererClass=&NiagaraMeshRendererProperties, rendererIndex=0}]
 }
 Foot_Small {
  emitterUpdateScript [
   &EmitterState(Inactive Response=::NewEnumerator1, Life Cycle Mode=::NewEnumerator1, Loop Duration=0.15, UseLoopDelay=false)
   &SpawnBurst_Instantaneous(Spawn Count=18, Use Spawn Probability=false)
  ]
  particleSpawnScript [
   /SetVariables(
    Particles.X_Grid:NiagaraFloat=&Lerp_Float(
     A=0
     B=18
     Alpha=&Frac_Float(Float=&Divide_Float(A=&MakeFloatFromInt(Int=&ReturnParticleID()), B=3))
    )
    Particles.Y_Grid:NiagaraFloat=&Multiply_Float(
     A=6
     B=&Floor_Float(Float:NiagaraFloat=&Divide_Float(A=&MakeFloatFromInt(Int=&ReturnParticleID()), B=6))
    )
   ) {isSetParameters=true, module=SetVariables_#2}
   &InitializeParticle(
    Color=&MakeCustomLinearColorFromBool(
     Bool=&RandomBool(Randomness Mode:ENiagaraRandomnessMode=::NewEnumerator0, Random Seed:NiagaraInt32=0, Fixed Random Seed:NiagaraBool=false)
     True Color=&MakeCustomLinearColorFromBool(
      Bool=&RandomBool(Randomness Mode:ENiagaraRandomnessMode=::NewEnumerator0, Random Seed:NiagaraInt32=0, Fixed Random Seed:NiagaraBool=false)
      True Color=<rgba 0.0344 1.6317 8 2>
      False Color=<rgba 0 0.6138 4 2>
     )
     False Color=<rgba 0.2604 3.093 15 2>
    )
    Lifetime=&UniformRangedFloat(Minimum=0.15, Maximum=0.15)
    Mesh Scale=&Multiply_Vector(
     A=<0.7 0.7 0.8>
     B=&VectorFromFloat(Value=&UniformRangedFloat(
       Minimum=&Multiply_Float(
        A=0.25
        B=&FloatFromCurve(
         FloatCurve=DataInterface() {propertyValues {Curve {keys [<at 0 1>,<at 1 1>]}}}
         CurveIndex=&Divide_Float(A=&VectorLength(Vector=$Engine.Owner.Velocity), B=9000)
        )
       )
       Maximum=&Multiply_Float(
        A=0.37
        B=&FloatFromCurve(
         FloatCurve=DataInterface() {propertyValues {Curve {keys [<at 0 1>,<at 0.9959 1>]}}}
         CurveIndex=&Divide_Float(A=&VectorLength(Vector=$Engine.Owner.Velocity), B=9000)
        )
       )
      ))
    )
    Module.Write Color=true
    Module.Write Lifetime=true
    Module.Write Mass=true
    Module.Write Material Random=true
    Module.Write Position=false
    Module.Write Scale=true
    Module.Write SpriteRotation=true
    Module.Write SpriteSize=false
    Position=&SimulationPosition()
    Sprite Rotation=&NormalizedAngleToDegrees(NormalizedAngle=&UniformRangedFloat(Minimum=0, Maximum=1))
   ) {module=InitializeParticle001}
   &FN_SetSpawnGroup_v2(Number of Spawn Groups=2)
  ]
  particleUpdateScript [
   &ParticleState()
   &SolveForcesAndVelocity(Clamp Velocity=false, Limit Acceleration=false)
   &ScaleColor(
    Scale RGB=&FN_FrontEndVector3Modifier(
     Gameplay=&MakeVectorFromLinearColorRGB(Color=<rgba 0 0.5347 0.7 1>)
     Front End=&MakeVectorFromLinearColorRGB(Color=$User.FrontEnd_color)
    )
    ScaleA=true
    ScaleRGB=true
   )
   &ScaleMeshSize(Scale Factor=&VectorFromFloat(Value=&FloatFromCurve(FloatCurve=DataInterface() {propertyValues {Curve {keys [<at 0.0027 0.5892>,<at 0.2489 0.7204>,<at 0.5021 0.75>,<at 0.7475 0.7204>,<at 0.9984 0.6433>]}}})))
   &SkeletalMeshLocation(
    Bone / Socket Index=$Particles.SpawnGroup
    Bone Sampling Mode=::NewEnumerator4
    Inherit Velocity=false
    Interpolate Mesh Sampling=false
    MaskSpawnGroup=false
    Sampled Orientation Coordinate Space=::Local
    Sampled Position Offset=&Add_Vector(A=&MakeVector(X=$Particles.X_Grid, Y=$Particles.Y_Grid, Z=0), B=<-6.5 -2 8>)
    Sampled Position Offset Coordinate Space:ENiagaraMeshTransforms=::NewEnumerator2
    Skeletal Mesh=$User.SKeletal_Mesh_Foot
   )
   &CameraOffset(Camera Offset Amount=30)
  ]
  rendererClasses [&NiagaraMeshRendererProperties]
  renderers [{rendererClass=&NiagaraMeshRendererProperties, rendererIndex=0}]
 }
 Foot_blast-off {
  emitterUpdateScript [
   &EmitterState(Inactive Response=::NewEnumerator1, Life Cycle Mode=::NewEnumerator1, Loop Duration=0.15, UseLoopDelay=false)
   &SpawnBurst_Instantaneous(Spawn Count=15, Use Spawn Probability=false)
  ]
  particleSpawnScript [
   /SetVariables(
    Particles.X_Grid:NiagaraFloat=&Lerp_Float(
     A=0
     B=17.5
     Alpha=&Frac_Float(Float=&Divide_Float(A=&MakeFloatFromInt(Int=&ReturnParticleID()), B=3))
    )
    Particles.Y_Grid:NiagaraFloat=&Multiply_Float(
     A=3
     B=&Floor_Float(Float:NiagaraFloat=&Divide_Float(A=&MakeFloatFromInt(Int=&ReturnParticleID()), B=3))
    )
   ) {isSetParameters=true, module=SetVariables_#2}
   /SetVariables(Particles.Z_Grid:NiagaraFloat=&Lerp_Float(
     A=0
     B=5
     Alpha=&Frac_Float(Float=&Divide_Float(A=&MakeFloatFromInt(Int=&ReturnParticleID()), B=6))
    )) {isSetParameters=true, module=SetVariables_#3}
   &InitializeParticle(
    Color=&MakeCustomLinearColorFromBool(
     Bool=&RandomBool(Randomness Mode:ENiagaraRandomnessMode=::NewEnumerator0, Random Seed:NiagaraInt32=0, Fixed Random Seed:NiagaraBool=false)
     True Color=&MakeCustomLinearColorFromBool(
      Bool=&RandomBool(Randomness Mode:ENiagaraRandomnessMode=::NewEnumerator0, Random Seed:NiagaraInt32=0, Fixed Random Seed:NiagaraBool=false)
      True Color=<rgba 0.0344 1.6317 8 2>
      False Color=<rgba 0 0.6138 4 2>
     )
     False Color=<rgba 0.2604 3.093 15 2>
    )
    Lifetime=&UniformRangedFloat(Minimum=0.15, Maximum=0.15)
    Mesh Scale=&Multiply_Vector(
     A=<0.7 0.7 0.8>
     B=&VectorFromFloat(Value=&UniformRangedFloat(
       Minimum=&Multiply_Float(
        A=0.25
        B=&FloatFromCurve(
         FloatCurve=DataInterface() {propertyValues {Curve {keys [<at 0 1>,<at 1 1>]}}}
         CurveIndex=&Divide_Float(A=&VectorLength(Vector=$Engine.Owner.Velocity), B=9000)
        )
       )
       Maximum=&Multiply_Float(
        A=0.37
        B=&FloatFromCurve(
         FloatCurve=DataInterface() {propertyValues {Curve {keys [<at 0 1>,<at 0.9959 1>]}}}
         CurveIndex=&Divide_Float(A=&VectorLength(Vector=$Engine.Owner.Velocity), B=9000)
        )
       )
      ))
    )
    Module.Write Color=true
    Module.Write Lifetime=true
    Module.Write Mass=true
    Module.Write Material Random=true
    Module.Write Position=false
    Module.Write Scale=true
    Module.Write SpriteRotation=true
    Module.Write SpriteSize=false
    Position=&SimulationPosition()
    Sprite Rotation=&NormalizedAngleToDegrees(NormalizedAngle=&UniformRangedFloat(Minimum=0, Maximum=1))
   ) {module=InitializeParticle001}
   &FN_SetSpawnGroup_v2(Number of Spawn Groups=2)
  ]
  particleUpdateScript [
   &ParticleState()
   &SolveForcesAndVelocity(Clamp Velocity=false, Limit Acceleration=false)
   &ScaleColor(
    Scale RGB=&FN_FrontEndVector3Modifier(
     Gameplay=&MakeVectorFromLinearColorRGB(Color=<rgba 0 0.5347 0.7 1>)
     Front End=&MakeVectorFromLinearColorRGB(Color=$User.FrontEnd_color)
    )
    ScaleA=true
    ScaleRGB=true
   ) {module=ScaleColor001}
   &ScaleMeshSize(Scale Factor=&VectorFromFloat(Value=&FloatFromCurve(FloatCurve=DataInterface() {propertyValues {Curve {keys [<at 0.0027 0.5892>,<at 0.2489 0.7204>,<at 0.5021 0.842>,<at 0.7475 0.7204>,<at 0.9984 0.6433>]}}}))) {enabled=false, module=ScaleMeshSize002}
   &ScaleMeshSize(Scale Factor=&VectorFromFloat(Value=&FloatFromCurve(FloatCurve=DataInterface() {propertyValues {Curve {keys [<at 0.0027 0.5892>,<at 0.2489 0.7204>,<at 0.5021 0.842>,<at 0.7475 0.7>,<at 0.9984 0.4>]}}}))) {enabled=false, module=ScaleMeshSize001}
   &ScaleMeshSize(Scale Factor=&VectorFromFloat(Value=&FloatFromCurve(FloatCurve=DataInterface() {propertyValues {Curve {keys [<at 0.0027 0.84>,<at 0.9984 0.3>]}}}))) {enabled=false}
   &ScaleMeshSize(Scale Factor=&VectorFromFloat(Value=&FloatFromCurve(FloatCurve=DataInterface() {propertyValues {Curve {keys [<at 0.0027 0.5892>,<at 0.2489 0.7204>,<at 0.5021 0.75>,<at 0.7475 0.7204>,<at 0.9984 0.6433>]}}}))) {module=ScaleMeshSize003}
   &SkeletalMeshLocation(
    Bone / Socket Index=$Particles.SpawnGroup
    Bone Sampling Mode=::NewEnumerator4
    Inherit Velocity=false
    Interpolate Mesh Sampling=false
    MaskSpawnGroup=false
    Sampled Orientation Coordinate Space=::Local
    Sampled Position Offset=&Add_Vector(
     A=&MakeVector(X=$Particles.X_Grid, Y=$Particles.Y_Grid, Z=$Particles.Z_Grid)
     B=&MakeVector(
      X=-6.5
      Y=-2
      Z=&FloatFromCurve(FloatCurve=DataInterface() {propertyValues {Curve {keys [<at 0 12>,<at 1 15>]}}})
     )
    )
    Sampled Position Offset Coordinate Space:ENiagaraMeshTransforms=::NewEnumerator2
    Skeletal Mesh=$User.SKeletal_Mesh_Foot
   )
   &CameraOffset(Camera Offset Amount=30)
  ]
  rendererClasses [&NiagaraMeshRendererProperties]
  renderers [{rendererClass=&NiagaraMeshRendererProperties, rendererIndex=0}]
 }
 Hand_Big {
  emitterUpdateScript [
   &EmitterState(Inactive Response=::NewEnumerator1, Life Cycle Mode=::NewEnumerator1, Loop Duration=0.15, UseLoopDelay=false)
   &SpawnBurst_Instantaneous(Spawn Count=18, Use Spawn Probability=false)
  ]
  particleSpawnScript [
   /SetVariables(
    Particles.X_Grid:NiagaraFloat=&Lerp_Float(
     A=0
     B=20
     Alpha=&Frac_Float(Float=&Divide_Float(A=&MakeFloatFromInt(Int=&ReturnParticleID()), B=3))
    )
    Particles.Y_Grid:NiagaraFloat=&Multiply_Float(
     A=7
     B=&Floor_Float(Float:NiagaraFloat=&Divide_Float(A=&MakeFloatFromInt(Int=&ReturnParticleID()), B=6))
    )
   ) {isSetParameters=true, module=SetVariables_#2}
   &InitializeParticle(
    Color=&MakeCustomLinearColorFromBool(
     Bool=&RandomBool(Randomness Mode:ENiagaraRandomnessMode=::NewEnumerator0, Random Seed:NiagaraInt32=0, Fixed Random Seed:NiagaraBool=false)
     True Color=&MakeCustomLinearColorFromBool(
      Bool=&RandomBool(Randomness Mode:ENiagaraRandomnessMode=::NewEnumerator0, Random Seed:NiagaraInt32=0, Fixed Random Seed:NiagaraBool=false)
      True Color=<rgba 0.0344 1.6317 8 2>
      False Color=<rgba 0 0.6138 4 2>
     )
     False Color=<rgba 0.2604 3.093 15 2>
    )
    Lifetime=&UniformRangedFloat(Minimum=0.15, Maximum=0.15)
    Mesh Scale=&Multiply_Vector(
     A=<0.7 0.7 0.7>
     B=&VectorFromFloat(Value=&UniformRangedFloat(
       Minimum=&Multiply_Float(
        A=0.4
        B=&FloatFromCurve(
         FloatCurve=DataInterface() {propertyValues {Curve {keys [<at 0 1>,<at 1 1>]}}}
         CurveIndex=&Divide_Float(A=&VectorLength(Vector=$Engine.Owner.Velocity), B=9000)
        )
       )
       Maximum=&Multiply_Float(
        A=0.55
        B=&FloatFromCurve(
         FloatCurve=DataInterface() {propertyValues {Curve {keys [<at 0 1>,<at 1 1>]}}}
         CurveIndex=&Divide_Float(A=&VectorLength(Vector=$Engine.Owner.Velocity), B=9000)
        )
       )
      ))
    )
    Module.Write Color=true
    Module.Write Lifetime=true
    Module.Write Mass=true
    Module.Write Material Random=true
    Module.Write Position=false
    Module.Write Scale=true
    Module.Write SpriteRotation=true
    Module.Write SpriteSize=true
    Position=&SimulationPosition()
    Sprite Size=&Vector2DFromFloat(Value=0)
   )
   &FN_SetSpawnGroup_v2(Number of Spawn Groups=2)
  ]
  particleUpdateScript [
   &ParticleState()
   &SolveForcesAndVelocity(Clamp Velocity=false, Limit Acceleration=false)
   &ScaleColor(
    Scale RGB=&FN_FrontEndVector3Modifier(
     Gameplay=&MakeVectorFromLinearColorRGB(Color=<rgba 0 0.5347 0.7 1>)
     Front End=&MakeVectorFromLinearColorRGB(Color=$User.FrontEnd_color)
    )
    ScaleA=true
    ScaleRGB=true
   )
   &ScaleMeshSize(Scale Factor=&VectorFromFloat(Value=&FloatFromCurve(FloatCurve=DataInterface() {propertyValues {Curve {keys [<at -0.0019 0.5372>,<at 0.2471 0.5372>,<at 0.4981 0.7>,<at 0.7498 0.5456>,<at 1.0019 0.5372>]}}})))
   &SkeletalMeshLocation(
    Bone / Socket Index=$Particles.SpawnGroup
    Bone Sampling Mode=::NewEnumerator4
    Inherit Velocity=false
    Interpolate Mesh Sampling=false
    MaskSpawnGroup=false
    Sampled Orientation Coordinate Space=::Local
    Sampled Position Offset=&Add_Vector(A=&MakeVector(X=$Particles.X_Grid, Y=$Particles.Y_Grid, Z=0), B=<-7 -3 0>)
    Sampled Position Offset Coordinate Space:ENiagaraMeshTransforms=::NewEnumerator2
    Skeletal Mesh=$User.SkeletalMesh
   )
   &ScaleMeshSizeBySpeed()
   &CameraOffset(Camera Offset Amount=15)
  ]
  rendererClasses [&NiagaraMeshRendererProperties]
  renderers [{rendererClass=&NiagaraMeshRendererProperties, rendererIndex=0}]
 }
 Hand_Small {
  emitterUpdateScript [
   &EmitterState(Inactive Response=::NewEnumerator1, Life Cycle Mode=::NewEnumerator1, Loop Duration=0.15, UseLoopDelay=false)
   &SpawnBurst_Instantaneous(Spawn Count=18, Use Spawn Probability=false)
  ]
  particleSpawnScript [
   /SetVariables(
    Particles.X_Grid:NiagaraFloat=&Lerp_Float(
     A=0
     B=18
     Alpha=&Frac_Float(Float=&Divide_Float(A=&MakeFloatFromInt(Int=&ReturnParticleID()), B=3))
    )
    Particles.Y_Grid:NiagaraFloat=&Multiply_Float(
     A=6
     B=&Floor_Float(Float:NiagaraFloat=&Divide_Float(A=&MakeFloatFromInt(Int=&ReturnParticleID()), B=6))
    )
   ) {isSetParameters=true, module=SetVariables_#2}
   &InitializeParticle(
    Color=&MakeCustomLinearColorFromBool(
     Bool=&RandomBool(Randomness Mode:ENiagaraRandomnessMode=::NewEnumerator0, Random Seed:NiagaraInt32=0, Fixed Random Seed:NiagaraBool=false)
     True Color=&MakeCustomLinearColorFromBool(
      Bool=&RandomBool(Randomness Mode:ENiagaraRandomnessMode=::NewEnumerator0, Random Seed:NiagaraInt32=0, Fixed Random Seed:NiagaraBool=false)
      True Color=<rgba 0.0344 1.6317 8 2>
      False Color=<rgba 0 0.6138 4 2>
     )
     False Color=<rgba 0.2604 3.093 15 2>
    )
    Lifetime=&UniformRangedFloat(Minimum=0.15, Maximum=0.15)
    Mesh Scale=&Multiply_Vector(
     A=<0.7 0.7 0.8>
     B=&VectorFromFloat(Value=&UniformRangedFloat(
       Minimum=&Multiply_Float(
        A=0.25
        B=&FloatFromCurve(
         FloatCurve=DataInterface() {propertyValues {Curve {keys [<at 0 1>,<at 1 1>]}}}
         CurveIndex=&Divide_Float(A=&VectorLength(Vector=$Engine.Owner.Velocity), B=9000)
        )
       )
       Maximum=&Multiply_Float(
        A=0.37
        B=&FloatFromCurve(
         FloatCurve=DataInterface() {propertyValues {Curve {keys [<at 0 1>,<at 0.9959 1>]}}}
         CurveIndex=&Divide_Float(A=&VectorLength(Vector=$Engine.Owner.Velocity), B=9000)
        )
       )
      ))
    )
    Module.Write Color=true
    Module.Write Lifetime=true
    Module.Write Mass=true
    Module.Write Material Random=true
    Module.Write Position=false
    Module.Write Scale=true
    Module.Write SpriteRotation=true
    Module.Write SpriteSize=false
    Position=&SimulationPosition()
    Sprite Rotation=&NormalizedAngleToDegrees(NormalizedAngle=&UniformRangedFloat(Minimum=0, Maximum=1))
   ) {module=InitializeParticle001}
   &FN_SetSpawnGroup_v2(Number of Spawn Groups=2)
  ]
  particleUpdateScript [
   &ParticleState()
   &SolveForcesAndVelocity(Clamp Velocity=false, Limit Acceleration=false)
   &ScaleColor(
    Scale RGB=&FN_FrontEndVector3Modifier(
     Gameplay=&MakeVectorFromLinearColorRGB(Color=<rgba 0 0.5347 0.7 1>)
     Front End=&MakeVectorFromLinearColorRGB(Color=$User.FrontEnd_color)
    )
    ScaleA=true
    ScaleRGB=true
   )
   &ScaleMeshSize(Scale Factor=&VectorFromFloat(Value=&FloatFromCurve(FloatCurve=DataInterface() {propertyValues {Curve {keys [<at 0.0027 0.5892>,<at 0.2489 0.7204>,<at 0.5021 0.75>,<at 0.7475 0.7204>,<at 0.9984 0.6433>]}}})))
   &SkeletalMeshLocation(
    Bone / Socket Index=$Particles.SpawnGroup
    Bone Sampling Mode=::NewEnumerator4
    Inherit Velocity=false
    Interpolate Mesh Sampling=false
    MaskSpawnGroup=false
    Sampled Orientation Coordinate Space=::Local
    Sampled Position Offset=&Add_Vector(A=&MakeVector(X=$Particles.X_Grid, Y=$Particles.Y_Grid, Z=0), B=<-6.5 -2 8>)
    Sampled Position Offset Coordinate Space:ENiagaraMeshTransforms=::NewEnumerator2
    Skeletal Mesh=$User.SkeletalMesh
   )
   &CameraOffset(Camera Offset Amount=15)
  ]
  rendererClasses [&NiagaraMeshRendererProperties]
  renderers [{rendererClass=&NiagaraMeshRendererProperties, rendererIndex=0}]
 }
 Hand_blast-off {
  emitterUpdateScript [
   &EmitterState(Inactive Response=::NewEnumerator1, Life Cycle Mode=::NewEnumerator1, Loop Duration=0.15, UseLoopDelay=false)
   &SpawnBurst_Instantaneous(Spawn Count=15, Use Spawn Probability=false)
  ]
  particleSpawnScript [
   /SetVariables(
    Particles.X_Grid:NiagaraFloat=&Lerp_Float(
     A=0
     B=17.5
     Alpha=&Frac_Float(Float=&Divide_Float(A=&MakeFloatFromInt(Int=&ReturnParticleID()), B=3))
    )
    Particles.Y_Grid:NiagaraFloat=&Multiply_Float(
     A=3
     B=&Floor_Float(Float:NiagaraFloat=&Divide_Float(A=&MakeFloatFromInt(Int=&ReturnParticleID()), B=3))
    )
   ) {isSetParameters=true, module=SetVariables_#2}
   /SetVariables(Particles.Z_Grid:NiagaraFloat=&Lerp_Float(
     A=0
     B=5
     Alpha=&Frac_Float(Float=&Divide_Float(A=&MakeFloatFromInt(Int=&ReturnParticleID()), B=6))
    )) {isSetParameters=true, module=SetVariables_#3}
   &InitializeParticle(
    Color=&MakeCustomLinearColorFromBool(
     Bool=&RandomBool(Randomness Mode:ENiagaraRandomnessMode=::NewEnumerator0, Random Seed:NiagaraInt32=0, Fixed Random Seed:NiagaraBool=false)
     True Color=&MakeCustomLinearColorFromBool(
      Bool=&RandomBool(Randomness Mode:ENiagaraRandomnessMode=::NewEnumerator0, Random Seed:NiagaraInt32=0, Fixed Random Seed:NiagaraBool=false)
      True Color=<rgba 0.0344 1.6317 8 2>
      False Color=<rgba 0 0.6138 4 2>
     )
     False Color=<rgba 0.2604 3.093 15 2>
    )
    Lifetime=&UniformRangedFloat(Minimum=0.15, Maximum=0.15)
    Mesh Scale=&Multiply_Vector(
     A=<0.7 0.7 0.8>
     B=&VectorFromFloat(Value=&UniformRangedFloat(
       Minimum=&Multiply_Float(
        A=0.25
        B=&FloatFromCurve(
         FloatCurve=DataInterface() {propertyValues {Curve {keys [<at 0 1>,<at 1 1>]}}}
         CurveIndex=&Divide_Float(A=&VectorLength(Vector=$Engine.Owner.Velocity), B=9000)
        )
       )
       Maximum=&Multiply_Float(
        A=0.37
        B=&FloatFromCurve(
         FloatCurve=DataInterface() {propertyValues {Curve {keys [<at 0 1>,<at 0.9959 1>]}}}
         CurveIndex=&Divide_Float(A=&VectorLength(Vector=$Engine.Owner.Velocity), B=9000)
        )
       )
      ))
    )
    Module.Write Color=true
    Module.Write Lifetime=true
    Module.Write Mass=true
    Module.Write Material Random=true
    Module.Write Position=false
    Module.Write Scale=true
    Module.Write SpriteRotation=true
    Module.Write SpriteSize=false
    Position=&SimulationPosition()
    Sprite Rotation=&NormalizedAngleToDegrees(NormalizedAngle=&UniformRangedFloat(Minimum=0, Maximum=1))
   ) {module=InitializeParticle001}
   &FN_SetSpawnGroup_v2(Number of Spawn Groups=2)
  ]
  particleUpdateScript [
   &ParticleState()
   &SolveForcesAndVelocity(Clamp Velocity=false, Limit Acceleration=false)
   &ScaleColor(
    Scale RGB=&FN_FrontEndVector3Modifier(
     Gameplay=&MakeVectorFromLinearColorRGB(Color=<rgba 0 0.5347 0.7 1>)
     Front End=&MakeVectorFromLinearColorRGB(Color=$User.FrontEnd_color)
    )
    ScaleA=true
    ScaleRGB=true
   ) {module=ScaleColor001}
   &ScaleMeshSize(Scale Factor=&VectorFromFloat(Value=&FloatFromCurve(FloatCurve=DataInterface() {propertyValues {Curve {keys [<at 0.0027 0.5892>,<at 0.2489 0.7204>,<at 0.5021 0.75>,<at 0.7475 0.7204>,<at 0.9984 0.6433>]}}})))
   &SkeletalMeshLocation(
    Bone / Socket Index=$Particles.SpawnGroup
    Bone Sampling Mode=::NewEnumerator4
    Inherit Velocity=false
    Interpolate Mesh Sampling=false
    MaskSpawnGroup=false
    Sampled Orientation Coordinate Space=::Local
    Sampled Position Offset=&Add_Vector(
     A=&MakeVector(X=$Particles.X_Grid, Y=$Particles.Y_Grid, Z=$Particles.Z_Grid)
     B=&MakeVector(
      X=-6.5
      Y=-2
      Z=&FloatFromCurve(FloatCurve=DataInterface() {propertyValues {Curve {keys [<at 0 12>,<at 1 15>]}}})
     )
    )
    Sampled Position Offset Coordinate Space:ENiagaraMeshTransforms=::NewEnumerator2
    Skeletal Mesh=$User.SkeletalMesh
   )
   &CameraOffset(Camera Offset Amount=15)
  ]
  rendererClasses [&NiagaraMeshRendererProperties]
  renderers [{rendererClass=&NiagaraMeshRendererProperties, rendererIndex=0}]
 }
 Ray {
  emitterSpawnScript [&FN_PreserveSkydiveUserVariables() {enabled=false, module=PreserveUserVariables}]
  emitterUpdateScript [
   &EmitterState(Inactive Response=::NewEnumerator1, Life Cycle Mode=::NewEnumerator1, Loop Duration=3, Scalability Mode=::NewEnumerator1, UseLoopDelay=false)
   &SpawnRate(SpawnRate=4, Use Spawn Probability=false)
   &SpawnRate(SpawnGroup=1, SpawnRate=4, Use Spawn Probability=false) {module=SpawnRate001}
  ]
  particleSpawnScript [
   /SetVariables(
    Particles.Dy_Darkness:NiagaraFloat=&UniformRangedFloat(Minimum=30, Maximum=1000)
    Particles.Dy_U_Tile:NiagaraFloat=&UniformRangedFloat(Minimum=0.5, Maximum=1)
    Particles.Dy_offsetUV:NiagaraFloat=&UniformRangedFloat(Minimum=0, Maximum=1)
    Particles.FallingSpeed:Vector3f=$Engine.Owner.Velocity
    Particles.PositionRandom:Vector3f=&UniformRangedVector(Minimum=<-2.5 -2.5 0>, Maximum=<2.5 2.5 0>)
   ) {isSetParameters=true}
   &InitializeParticle(
    Color=<rgba 0 0.9183 3 1>
    Lifetime=&UniformRangedFloat(Minimum=1, Maximum=1)
    Mesh Scale=&Multiply_Vector(
     A=<1 1 1>
     B=&VectorFromFloat(Value=&UniformRangedFloat(Minimum=0.2, Maximum=1))
    )
    Module.Write Color=true
    Module.Write Lifetime=true
    Module.Write Mass=true
    Module.Write Material Random=true
    Module.Write Position=false
    Module.Write Scale=false
    Module.Write SpriteRotation=true
    Module.Write SpriteSize=true
    Position=&SimulationPosition()
    Sprite Rotation=90
    Sprite Size=&MakeVector2D(
     X=&Multiply_Float(
      A=15
      B=&Lerp_Float(A=&FN_FrontEndFloatModifier(Gameplay=0.3, Front End=1.3), B=1.3, Alpha=$Particles.Dy_U_Tile)
     )
     Y=&Multiply_Float(
      A=&UniformRangedFloat(
       Minimum=&FN_FrontEndFloatModifier(Gameplay=200, Front End=300)
       Maximum=&FN_FrontEndFloatModifier(Gameplay=250, Front End=400)
      )
      B=&FloatFromCurve(
       FloatCurve=DataInterface() {propertyValues {Curve {keys [<at 0 1>,<at 1 2.5>]}}}
       CurveIndex=&Divide_Float(A=&VectorLength(Vector=$Particles.FallingSpeed), B=9000)
      )
     )
    )
   )
   &FN_SetSpawnGroup_v2(Number of Spawn Groups=4)
   &AddVelocity(Velocity=<0 0 1>)
   &DynamicMaterialParameters(
    Index 0 Param 1=$Particles.Dy_offsetUV
    Index 0 Param 2=$Particles.Dy_U_Tile
    Index 0 Param 3=$Particles.Dy_Darkness
    Index 0 Param 4=1
    Module.Param0WriteEnabled=true
    Module.Param1WriteEnabled=true
    Module.Param2WriteEnabled=true
    Module.Param3WriteEnabled=true
   )
   &CameraOffset(Camera Offset Amount=&UniformRangedFloat(Minimum=-5, Maximum=5))
  ]
  particleUpdateScript [
   &SkeletalMeshLocation(
    Bone / Socket Index=$Particles.SpawnGroup
    Bone Sampling Mode=::NewEnumerator4
    Inherit Velocity=false
    MaskSpawnGroup=false
    Sampled Position Offset=<0 0 -15>
    Skeletal Mesh=$User.SkeletalMesh
   )
   &ParticleState()
   &SolveForcesAndVelocity(Clamp Velocity=false, Limit Acceleration=false)
   &ScaleColor(
    Scale Alpha=&Multiply_Float(
     A=&FloatFromCurve(FloatCurve=DataInterface() {propertyValues {Curve {keys [<at 0 0 RCIM_Cubic>,<at 0.37 0.1417 RCIM_Cubic>,<at 1 0>]}}})
     B=&FN_FrontEndFloatModifier(Gameplay=1, Front End=0.75)
    )
    ScaleA=true
    ScaleRGB=true
   )
  ]
  rendererClasses [&NiagaraSpriteRendererProperties]
  renderers [{rendererClass=&NiagaraSpriteRendererProperties}]
 }
 Square_close_Attach {
  emitterUpdateScript [
   &EmitterState(Inactive Response=::NewEnumerator1, Life Cycle Mode=::NewEnumerator1, Loop Duration=1, UseLoopDelay=false)
   &SpawnRate(SpawnRate=&FN_FrontEndFloatModifier(Gameplay=9, Front End=9), Use Spawn Probability=false)
   &SpawnRate(
    SpawnGroup=1
    SpawnRate=&FN_FrontEndFloatModifier(Gameplay=9, Front End=9)
    Use Spawn Probability=false
   ) {module=SpawnRate001}
   &SpawnRate(
    SpawnGroup=2
    SpawnRate=&FN_FrontEndFloatModifier(Gameplay=9, Front End=9)
    Use Spawn Probability=false
   ) {module=SpawnRate002}
   &SpawnRate(
    SpawnGroup=3
    SpawnRate=&FN_FrontEndFloatModifier(Gameplay=9, Front End=9)
    Use Spawn Probability=false
   ) {module=SpawnRate003}
   &SpawnRate(
    SpawnGroup=4
    SpawnRate=&FN_FrontEndFloatModifier(Gameplay=9, Front End=9)
    Use Spawn Probability=false
   ) {module=SpawnRate004}
   &SpawnRate(
    SpawnGroup=5
    SpawnRate=&FN_FrontEndFloatModifier(Gameplay=9, Front End=9)
    Use Spawn Probability=false
   ) {module=SpawnRate005}
   &SpawnRate(
    SpawnGroup=6
    SpawnRate=&FN_FrontEndFloatModifier(Gameplay=9, Front End=9)
    Use Spawn Probability=false
   ) {module=SpawnRate006}
   &SpawnRate(
    SpawnGroup=7
    SpawnRate=&FN_FrontEndFloatModifier(Gameplay=9, Front End=9)
    Use Spawn Probability=false
   ) {module=SpawnRate007}
   &SpawnRate(
    SpawnGroup=8
    SpawnRate=&FN_FrontEndFloatModifier(Gameplay=9, Front End=9)
    Use Spawn Probability=false
   ) {module=SpawnRate008}
  ]
  particleSpawnScript [
   &InitializeParticle(
    Color=&MakeCustomLinearColorFromBool(
     Bool=&RandomBool(Randomness Mode:ENiagaraRandomnessMode=::NewEnumerator0, Random Seed:NiagaraInt32=0, Fixed Random Seed:NiagaraBool=false)
     True Color=&MakeCustomLinearColorFromBool(
      Bool=&RandomBool(Randomness Mode:ENiagaraRandomnessMode=::NewEnumerator0, Random Seed:NiagaraInt32=0, Fixed Random Seed:NiagaraBool=false)
      True Color=<rgba 0.043 2.0396 10 2>
      False Color=<rgba 0 0.6138 4 2>
     )
     False Color=<rgba 0.2604 3.093 15 2>
    )
    Lifetime=&UniformRangedFloat(Minimum=0.2, Maximum=0.4)
    Mesh Scale=&Multiply_Vector(
     A=<0.7 0.7 0.7>
     B=&VectorFromFloat(Value=&UniformRangedFloat(
       Minimum=&Multiply_Float(
        A=0.25
        B=&FloatFromCurve(
         FloatCurve=DataInterface() {propertyValues {Curve {keys [<at 0 1>,<at 1 1>]}}}
         CurveIndex=&Divide_Float(A=&VectorLength(Vector=$Engine.Owner.Velocity), B=9000)
        )
       )
       Maximum=&Multiply_Float(
        A=0.27
        B=&FloatFromCurve(
         FloatCurve=DataInterface() {propertyValues {Curve {keys [<at 0 1>,<at 1 1.2>]}}}
         CurveIndex=&Divide_Float(A=&VectorLength(Vector=$Engine.Owner.Velocity), B=9000)
        )
       )
      ))
    )
    Module.Write Color=true
    Module.Write Lifetime=true
    Module.Write Mass=true
    Module.Write Material Random=true
    Module.Write Position=false
    Module.Write Scale=true
    Module.Write SpriteRotation=true
    Module.Write SpriteSize=false
    Position=&SimulationPosition()
    Sprite Rotation=&NormalizedAngleToDegrees(NormalizedAngle=&UniformRangedFloat(Minimum=0, Maximum=1))
   ) {module=InitializeParticle002}
   &FN_SetSpawnGroup_v2(Number of Spawn Groups=4)
   &SkeletalMeshLocation(
    Bone / Socket Index=$Particles.SpawnGroup
    Bone Sampling Mode=::NewEnumerator4
    Inherit Velocity=false
    Interpolate Mesh Sampling=false
    MaskSpawnGroup=true
    Sampled Orientation Coordinate Space=::World
    Sampled Position Offset=&MakeVector(X=-0.5, Y=4, Z=$User.Height_Location)
    Sampled Position Offset Coordinate Space:ENiagaraMeshTransforms=::NewEnumerator2
    Skeletal Mesh=$User.SkeletalMesh
   )
   &SkeletalMeshLocation(
    Bone / Socket Index=$Particles.SpawnGroup
    Bone Sampling Mode=::NewEnumerator4
    Inherit Velocity=false
    Interpolate Mesh Sampling=false
    MaskSpawnGroup=true
    Sampled Orientation Coordinate Space=::World
    Sampled Position Offset=&MakeVector(X=-6.5, Y=10, Z=$User.Height_Location)
    Sampled Position Offset Coordinate Space:ENiagaraMeshTransforms=::NewEnumerator2
    Skeletal Mesh=$User.SkeletalMesh
    Spawn Group Mask=1
   ) {module=SkeletalMeshLocation001}
   &SkeletalMeshLocation(
    Bone / Socket Index=$Particles.SpawnGroup
    Bone Sampling Mode=::NewEnumerator4
    Inherit Velocity=false
    Interpolate Mesh Sampling=false
    MaskSpawnGroup=true
    Sampled Orientation Coordinate Space=::World
    Sampled Position Offset=&MakeVector(X=-6.5, Y=4, Z=$User.Height_Location)
    Sampled Position Offset Coordinate Space:ENiagaraMeshTransforms=::NewEnumerator2
    Skeletal Mesh=$User.SkeletalMesh
    Spawn Group Mask=2
   ) {module=SkeletalMeshLocation002}
   &SkeletalMeshLocation(
    Bone / Socket Index=$Particles.SpawnGroup
    Bone Sampling Mode=::NewEnumerator4
    Inherit Velocity=false
    Interpolate Mesh Sampling=false
    MaskSpawnGroup=true
    Sampled Orientation Coordinate Space=::World
    Sampled Position Offset=&MakeVector(X=5.5, Y=4, Z=$User.Height_Location)
    Sampled Position Offset Coordinate Space:ENiagaraMeshTransforms=::NewEnumerator2
    Skeletal Mesh=$User.SkeletalMesh
    Spawn Group Mask=3
   ) {module=SkeletalMeshLocation003}
   &SkeletalMeshLocation(
    Bone / Socket Index=$Particles.SpawnGroup
    Bone Sampling Mode=::NewEnumerator4
    Inherit Velocity=false
    Interpolate Mesh Sampling=false
    MaskSpawnGroup=true
    Sampled Orientation Coordinate Space=::World
    Sampled Position Offset=&MakeVector(X=5.5, Y=-2, Z=$User.Height_Location)
    Sampled Position Offset Coordinate Space:ENiagaraMeshTransforms=::NewEnumerator2
    Skeletal Mesh=$User.SkeletalMesh
    Spawn Group Mask=4
   ) {module=SkeletalMeshLocation004}
   &SkeletalMeshLocation(
    Bone / Socket Index=$Particles.SpawnGroup
    Bone Sampling Mode=::NewEnumerator4
    Inherit Velocity=false
    Interpolate Mesh Sampling=false
    MaskSpawnGroup=true
    Sampled Orientation Coordinate Space=::World
    Sampled Position Offset=&MakeVector(X=-6.5, Y=-2, Z=$User.Height_Location)
    Sampled Position Offset Coordinate Space:ENiagaraMeshTransforms=::NewEnumerator2
    Skeletal Mesh=$User.SkeletalMesh
    Spawn Group Mask=5
   ) {module=SkeletalMeshLocation005}
   &SkeletalMeshLocation(
    Bone / Socket Index=$Particles.SpawnGroup
    Bone Sampling Mode=::NewEnumerator4
    Inherit Velocity=false
    Interpolate Mesh Sampling=false
    MaskSpawnGroup=true
    Sampled Orientation Coordinate Space=::World
    Sampled Position Offset=&MakeVector(X=-0.5, Y=-2, Z=$User.Height_Location)
    Sampled Position Offset Coordinate Space:ENiagaraMeshTransforms=::NewEnumerator2
    Skeletal Mesh=$User.SkeletalMesh
    Spawn Group Mask=6
   ) {module=SkeletalMeshLocation006}
   &SkeletalMeshLocation(
    Bone / Socket Index=$Particles.SpawnGroup
    Bone Sampling Mode=::NewEnumerator4
    Inherit Velocity=false
    Interpolate Mesh Sampling=false
    MaskSpawnGroup=true
    Sampled Orientation Coordinate Space=::World
    Sampled Position Offset=&MakeVector(X=-0.5, Y=10, Z=$User.Height_Location)
    Sampled Position Offset Coordinate Space:ENiagaraMeshTransforms=::NewEnumerator2
    Skeletal Mesh=$User.SkeletalMesh
    Spawn Group Mask=7
   ) {module=SkeletalMeshLocation007}
   &SkeletalMeshLocation(
    Bone / Socket Index=$Particles.SpawnGroup
    Bone Sampling Mode=::NewEnumerator4
    Inherit Velocity=false
    Interpolate Mesh Sampling=false
    MaskSpawnGroup=true
    Sampled Orientation Coordinate Space=::World
    Sampled Position Offset=&MakeVector(X=5.5, Y=10, Z=$User.Height_Location)
    Sampled Position Offset Coordinate Space:ENiagaraMeshTransforms=::NewEnumerator2
    Skeletal Mesh=$User.SkeletalMesh
    Spawn Group Mask=8
   ) {module=SkeletalMeshLocation008}
   &AddVelocity(Velocity=&Multiply_Vector(
     A=&UniformRangedVector(Minimum=<0 0 30>, Maximum=<0 0 300>)
     B=&VectorFromFloat(Value=&FloatFromCurve(
       FloatCurve=DataInterface() {propertyValues {Curve {keys [<at 0 1>,<at 1 4>]}}}
       CurveIndex=&Divide_Float(A=&VectorLength(Vector=$Engine.Owner.Velocity), B=9000)
      ))
    ))
  ]
  particleUpdateScript [
   &ParticleState()
   &Drag(Drag=2, Use Linear Drag=true, Use Rotational Drag=false)
   &SolveForcesAndVelocity(Clamp Velocity=false, Limit Acceleration=false)
   &ScaleColor(
    Scale Alpha=&FloatFromCurve(FloatCurve=DataInterface() {propertyValues {Curve {keys [<at 0 0>,<at 0.25 1>,<at 0.5 1>,<at 1 0>]}}})
    ScaleA=true
    ScaleRGB=true
   ) {enabled=false}
   &ScaleColor(
    Scale Alpha=&FloatFromCurve(FloatCurve=DataInterface() {propertyValues {Curve {keys [<at 0 0>,<at 0.1 1>,<at 0.9 1>,<at 1 0>]}}})
    Scale RGB=&FN_FrontEndVector3Modifier(Gameplay=<1 1 1>, Front End=&MakeVectorFromLinearColorRGB(Color=<rgba 2.5 0.6315 0.4322 1>))
    ScaleA=true
    ScaleRGB=true
   ) {enabled=false, module=ScaleColor001}
   &ScaleColor(
    Scale RGB=&FN_FrontEndVector3Modifier(
     Gameplay=&MakeVectorFromLinearColorRGB(Color=<rgba 0 0.5347 0.7 1>)
     Front End=&MakeVectorFromLinearColorRGB(Color=$User.FrontEnd_color)
    )
    ScaleA=true
    ScaleRGB=true
   ) {module=ScaleColor002}
   &ScaleMeshSize(Scale Factor=&VectorFromFloat(Value=&FloatFromCurve(FloatCurve=DataInterface() {propertyValues {Curve {keys [<at 0 1>,<at 0.9998 0.2>]}}}))) {module=ScaleMeshSize002}
   &CameraOffset(Camera Offset Amount=&UniformRangedFloat(Minimum=15, Maximum=30))
  ]
  rendererClasses [&NiagaraMeshRendererProperties]
  renderers [{rendererClass=&NiagaraMeshRendererProperties, rendererIndex=0}]
 }
 Square_far {
  emitterUpdateScript [
   &EmitterState(Inactive Response=::NewEnumerator1, Life Cycle Mode=::NewEnumerator1, Loop Duration=1, UseLoopDelay=false)
   &SpawnRate(SpawnRate=&FN_FrontEndFloatModifier(Gameplay=5, Front End=5), Use Spawn Probability=false)
   &SpawnRate(
    SpawnGroup=1
    SpawnRate=&FN_FrontEndFloatModifier(Gameplay=5, Front End=5)
    Use Spawn Probability=false
   ) {module=SpawnRate001}
   &SpawnRate(
    SpawnGroup=2
    SpawnRate=&FN_FrontEndFloatModifier(Gameplay=5, Front End=5)
    Use Spawn Probability=false
   ) {module=SpawnRate002}
   &SpawnRate(
    SpawnGroup=3
    SpawnRate=&FN_FrontEndFloatModifier(Gameplay=5, Front End=5)
    Use Spawn Probability=false
   ) {module=SpawnRate003}
   &SpawnRate(
    SpawnGroup=4
    SpawnRate=&FN_FrontEndFloatModifier(Gameplay=5, Front End=5)
    Use Spawn Probability=false
   ) {module=SpawnRate004}
   &SpawnRate(
    SpawnGroup=5
    SpawnRate=&FN_FrontEndFloatModifier(Gameplay=5, Front End=5)
    Use Spawn Probability=false
   ) {module=SpawnRate005}
   &SpawnRate(
    SpawnGroup=6
    SpawnRate=&FN_FrontEndFloatModifier(Gameplay=5, Front End=5)
    Use Spawn Probability=false
   ) {module=SpawnRate006}
   &SpawnRate(
    SpawnGroup=7
    SpawnRate=&FN_FrontEndFloatModifier(Gameplay=5, Front End=5)
    Use Spawn Probability=false
   ) {module=SpawnRate007}
   &SpawnRate(
    SpawnGroup=8
    SpawnRate=&FN_FrontEndFloatModifier(Gameplay=5, Front End=5)
    Use Spawn Probability=false
   ) {module=SpawnRate008}
  ]
  particleSpawnScript [
   &InitializeParticle(
    Color=&MakeCustomLinearColorFromBool(
     Bool=&RandomBool(Randomness Mode:ENiagaraRandomnessMode=::NewEnumerator0, Random Seed:NiagaraInt32=0, Fixed Random Seed:NiagaraBool=false)
     True Color=&MakeCustomLinearColorFromBool(
      Bool=&RandomBool(Randomness Mode:ENiagaraRandomnessMode=::NewEnumerator0, Random Seed:NiagaraInt32=0, Fixed Random Seed:NiagaraBool=false)
      True Color=<rgba 0.043 2.0396 10 2>
      False Color=<rgba 0 0.6138 4 2>
     )
     False Color=<rgba 0.2604 3.093 15 2>
    )
    Lifetime=&UniformRangedFloat(Minimum=0.3, Maximum=0.4)
    Mesh Scale=&Multiply_Vector(
     A=<0.7 0.7 0.7>
     B=&VectorFromFloat(Value=&UniformRangedFloat(
       Minimum=&Multiply_Float(
        A=0.15
        B=&FloatFromCurve(
         FloatCurve=DataInterface() {propertyValues {Curve {keys [<at 0 1>,<at 1 1>]}}}
         CurveIndex=&Divide_Float(A=&VectorLength(Vector=$Engine.Owner.Velocity), B=9000)
        )
       )
       Maximum=&Multiply_Float(
        A=0.25
        B=&FloatFromCurve(
         FloatCurve=DataInterface() {propertyValues {Curve {keys [<at 0 1>,<at 1 1.2>]}}}
         CurveIndex=&Divide_Float(A=&VectorLength(Vector=$Engine.Owner.Velocity), B=9000)
        )
       )
      ))
    )
    Module.Write Color=true
    Module.Write Lifetime=true
    Module.Write Mass=true
    Module.Write Material Random=true
    Module.Write Position=false
    Module.Write Scale=true
    Module.Write SpriteRotation=true
    Module.Write SpriteSize=false
    Position=&SimulationPosition()
    Sprite Rotation=&NormalizedAngleToDegrees(NormalizedAngle=&UniformRangedFloat(Minimum=0, Maximum=1))
   ) {module=InitializeParticle002}
   &FN_SetSpawnGroup_v2(Number of Spawn Groups=4)
   &SkeletalMeshLocation(
    Bone / Socket Index=$Particles.SpawnGroup
    Bone Sampling Mode=::NewEnumerator4
    Inherit Velocity=false
    Interpolate Mesh Sampling=false
    MaskSpawnGroup=true
    Sampled Orientation Coordinate Space=::World
    Sampled Position Offset=&MakeVector(X=-0.5, Y=4, Z=$User.Height_Location)
    Sampled Position Offset Coordinate Space:ENiagaraMeshTransforms=::NewEnumerator2
    Skeletal Mesh=$User.SkeletalMesh
   )
   &SkeletalMeshLocation(
    Bone / Socket Index=$Particles.SpawnGroup
    Bone Sampling Mode=::NewEnumerator4
    Inherit Velocity=false
    Interpolate Mesh Sampling=false
    MaskSpawnGroup=true
    Sampled Orientation Coordinate Space=::World
    Sampled Position Offset=&MakeVector(X=-6.5, Y=10, Z=$User.Height_Location)
    Sampled Position Offset Coordinate Space:ENiagaraMeshTransforms=::NewEnumerator2
    Skeletal Mesh=$User.SkeletalMesh
    Spawn Group Mask=1
   ) {module=SkeletalMeshLocation001}
   &SkeletalMeshLocation(
    Bone / Socket Index=$Particles.SpawnGroup
    Bone Sampling Mode=::NewEnumerator4
    Inherit Velocity=false
    Interpolate Mesh Sampling=false
    MaskSpawnGroup=true
    Sampled Orientation Coordinate Space=::World
    Sampled Position Offset=&MakeVector(X=-6.5, Y=4, Z=$User.Height_Location)
    Sampled Position Offset Coordinate Space:ENiagaraMeshTransforms=::NewEnumerator2
    Skeletal Mesh=$User.SkeletalMesh
    Spawn Group Mask=2
   ) {module=SkeletalMeshLocation002}
   &SkeletalMeshLocation(
    Bone / Socket Index=$Particles.SpawnGroup
    Bone Sampling Mode=::NewEnumerator4
    Inherit Velocity=false
    Interpolate Mesh Sampling=false
    MaskSpawnGroup=true
    Sampled Orientation Coordinate Space=::World
    Sampled Position Offset=&MakeVector(X=5.5, Y=4, Z=$User.Height_Location)
    Sampled Position Offset Coordinate Space:ENiagaraMeshTransforms=::NewEnumerator2
    Skeletal Mesh=$User.SkeletalMesh
    Spawn Group Mask=3
   ) {module=SkeletalMeshLocation003}
   &SkeletalMeshLocation(
    Bone / Socket Index=$Particles.SpawnGroup
    Bone Sampling Mode=::NewEnumerator4
    Inherit Velocity=false
    Interpolate Mesh Sampling=false
    MaskSpawnGroup=true
    Sampled Orientation Coordinate Space=::World
    Sampled Position Offset=&MakeVector(X=5.5, Y=-2, Z=$User.Height_Location)
    Sampled Position Offset Coordinate Space:ENiagaraMeshTransforms=::NewEnumerator2
    Skeletal Mesh=$User.SkeletalMesh
    Spawn Group Mask=4
   ) {module=SkeletalMeshLocation004}
   &SkeletalMeshLocation(
    Bone / Socket Index=$Particles.SpawnGroup
    Bone Sampling Mode=::NewEnumerator4
    Inherit Velocity=false
    Interpolate Mesh Sampling=false
    MaskSpawnGroup=true
    Sampled Orientation Coordinate Space=::World
    Sampled Position Offset=&MakeVector(X=-6.5, Y=-2, Z=$User.Height_Location)
    Sampled Position Offset Coordinate Space:ENiagaraMeshTransforms=::NewEnumerator2
    Skeletal Mesh=$User.SkeletalMesh
    Spawn Group Mask=5
   ) {module=SkeletalMeshLocation005}
   &SkeletalMeshLocation(
    Bone / Socket Index=$Particles.SpawnGroup
    Bone Sampling Mode=::NewEnumerator4
    Inherit Velocity=false
    Interpolate Mesh Sampling=false
    MaskSpawnGroup=true
    Sampled Orientation Coordinate Space=::World
    Sampled Position Offset=&MakeVector(X=-0.5, Y=-2, Z=$User.Height_Location)
    Sampled Position Offset Coordinate Space:ENiagaraMeshTransforms=::NewEnumerator2
    Skeletal Mesh=$User.SkeletalMesh
    Spawn Group Mask=6
   ) {module=SkeletalMeshLocation006}
   &SkeletalMeshLocation(
    Bone / Socket Index=$Particles.SpawnGroup
    Bone Sampling Mode=::NewEnumerator4
    Inherit Velocity=false
    Interpolate Mesh Sampling=false
    MaskSpawnGroup=true
    Sampled Orientation Coordinate Space=::World
    Sampled Position Offset=&MakeVector(X=-0.5, Y=10, Z=$User.Height_Location)
    Sampled Position Offset Coordinate Space:ENiagaraMeshTransforms=::NewEnumerator2
    Skeletal Mesh=$User.SkeletalMesh
    Spawn Group Mask=7
   ) {module=SkeletalMeshLocation007}
   &SkeletalMeshLocation(
    Bone / Socket Index=$Particles.SpawnGroup
    Bone Sampling Mode=::NewEnumerator4
    Inherit Velocity=false
    Interpolate Mesh Sampling=false
    MaskSpawnGroup=true
    Sampled Orientation Coordinate Space=::World
    Sampled Position Offset=&MakeVector(X=5.5, Y=10, Z=$User.Height_Location)
    Sampled Position Offset Coordinate Space:ENiagaraMeshTransforms=::NewEnumerator2
    Skeletal Mesh=$User.SkeletalMesh
    Spawn Group Mask=8
   ) {module=SkeletalMeshLocation008}
   &AddVelocity(Velocity=&Multiply_Vector(
     A=&UniformRangedVector(Minimum=<0 0 200>, Maximum=<1 1 400>)
     B=&VectorFromFloat(Value=&FloatFromCurve(
       FloatCurve=DataInterface() {propertyValues {Curve {keys [<at 0 1>,<at 1 5>]}}}
       CurveIndex=&Divide_Float(A=&VectorLength(Vector=$Engine.Owner.Velocity), B=9000)
      ))
    ))
  ]
  particleUpdateScript [
   &ParticleState()
   &Drag(Drag=2, Use Linear Drag=true, Use Rotational Drag=false)
   &SolveForcesAndVelocity(Clamp Velocity=false, Limit Acceleration=false)
   &ScaleColor(
    Scale Alpha=&FloatFromCurve(FloatCurve=DataInterface() {propertyValues {Curve {keys [<at 0 0>,<at 0.25 1>,<at 0.5 1>,<at 1 0>]}}})
    ScaleA=true
    ScaleRGB=true
   ) {enabled=false}
   &ScaleColor(
    Scale Alpha=&FloatFromCurve(FloatCurve=DataInterface() {propertyValues {Curve {keys [<at 0 0>,<at 0.1 1>,<at 0.9 1>,<at 1 0>]}}})
    Scale RGB=&FN_FrontEndVector3Modifier(Gameplay=<1 1 1>, Front End=&MakeVectorFromLinearColorRGB(Color=<rgba 2.5 0.6315 0.4322 1>))
    ScaleA=true
    ScaleRGB=true
   ) {enabled=false, module=ScaleColor001}
   &ScaleColor(
    Scale RGB=&FN_FrontEndVector3Modifier(
     Gameplay=&MakeVectorFromLinearColorRGB(Color=<rgba 0 0.5347 0.7 1>)
     Front End=&MakeVectorFromLinearColorRGB(Color=$User.FrontEnd_color)
    )
    ScaleA=true
    ScaleRGB=true
   ) {module=ScaleColor002}
   &ScaleMeshSize(Scale Factor=&VectorFromFloat(Value=&FloatFromCurve(FloatCurve=DataInterface() {propertyValues {Curve {keys [{arriveTangent=-1.3985, interpMode=RCIM_Cubic, leaveTangent=-1.3985, tangentMode=RCTM_User, time=0, value=1}, <at 0.9998 0.2>]}}}))) {module=ScaleMeshSize002}
   &CameraOffset()
  ]
  rendererClasses [&NiagaraMeshRendererProperties]
  renderers [{rendererClass=&NiagaraMeshRendererProperties, rendererIndex=0}]
 }
}
summary {
 systemName=NS_Contrail_Turbo
 emitters [
  {emitterName=Square_far, rendererClasses [&NiagaraMeshRendererProperties]}
  {emitterName=Hand_Big, rendererClasses [&NiagaraMeshRendererProperties]}
  {emitterName=Hand_Small, rendererClasses [&NiagaraMeshRendererProperties]}
  {emitterName=Ray, rendererClasses [&NiagaraSpriteRendererProperties]}
  {emitterName=Hand_blast-off, rendererClasses [&NiagaraMeshRendererProperties]}
  {
   emitterName=FN_Contrails_AttachedParticle001
   rendererClasses [&NiagaraSpriteRendererProperties, &NiagaraLightRendererProperties]
  }
  {emitterName=Foot_Big, rendererClasses [&NiagaraMeshRendererProperties]}
  {emitterName=Foot_Small, rendererClasses [&NiagaraMeshRendererProperties]}
  {emitterName=Foot_blast-off, rendererClasses [&NiagaraMeshRendererProperties]}
  {emitterName=Square_close_Attach, rendererClasses [&NiagaraMeshRendererProperties]}
 ]
 userVariables [
  User.bIsFrontEnd {type=NiagaraBool}
  User.bIsFrontEndPreview {type=NiagaraBool}
  User.FrontEnd_color {type=LinearColor, description="", defaultValue {struct=LinearColor, value=<rgba 1 0.6743 0.6445 1>}}
  User.Height_Location {type=NiagaraFloat, defaultValue=17, description=""}
  User.SKeletal_Mesh_Foot {
   type=NiagaraDataInterfaceSkeletalMesh
   defaultValue {
    struct=NiagaraExt_VariableValue_DataInterface
    value {dataInterface=&self:NiagaraDataInterfaceSkeletalMesh_1}
   }
  }
  User.SkeletalMesh {
   type=NiagaraDataInterfaceSkeletalMesh
   defaultValue {
    struct=NiagaraExt_VariableValue_DataInterface
    value {dataInterface=&self:NiagaraDataInterfaceSkeletalMesh_0}
   }
  }
 ]
}
